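; Parent Blueprint for Season 24 electronics crate quest prop actors, with a configurable crate material.
; The construction script checks whether the pre-interaction mesh component equals the CrateMeshClosedA variable to choose material set A or B.
; A bool then selects the variant or original material, applied to slot 0 of both pre- and post-interaction meshes.
; After interaction, it runs the parent VFX and audio, then hides the quest icon's minimap visibility.
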(event EventBeginPlay
  (|CallParentFunction))

(event Collision|EventActorBeginOverlap
  (params (OtherActor :type Actor%20Object%20Reference))
  (|CallParentFunction OtherActor) ; This node is disabled and will not be called.\nDrag off pins to build functionality.
)

(event EventTick
  (params (DeltaSeconds :type Float%20%28single-precision%29))
  (|CallParentFunction DeltaSeconds) ; This node is disabled and will not be called.\nDrag off pins to build functionality.
)

(fn PostInteractionVFX&Audio
  (params (PlayerController :type Fort%20Player%20Controller%20Object%20Reference))
  (|CallParentFunction PlayerController)
  (MiniMapData|SetVisibility (Variables|BuildingGameplayActorPropQuest|GetQuestIconComponent) false))

(fn ConstructionScript
  ;[0 0 0] Set the material to display on the crate
  (|CallParentFunction)
  (if (== (Variables|Components|GetSM_Quest_PreInteraction) (Variables|Design-CrateCustomization|GetCrateMeshClosedA))
    ;;[0.2188 0.2188 0.2188] Crate Mesh A
    (if (Variables|Design-CrateCustomization|GetbUseVariantMaterial)
      (Rendering|Material|SetMaterial (Variables|Components|GetSM_Quest_PreInteraction) 0 (Variables|Design-CrateCustomization|GetVariantMaterialA))
      (Rendering|Material|SetMaterial (Variables|Components|GetSM_Quest_PostInteraction) 0 (Variables|Design-CrateCustomization|GetVariantMaterialA))
      (else
        (Rendering|Material|SetMaterial (Variables|Components|GetSM_Quest_PreInteraction) 0 (Variables|Design-CrateCustomization|GetOriginalMaterialA))
        (Rendering|Material|SetMaterial (Variables|Components|GetSM_Quest_PostInteraction) 0 (Variables|Design-CrateCustomization|GetOriginalMaterialA))))
    (elif (Variables|Design-CrateCustomization|GetbUseVariantMaterial)
      ;;[0.224 0.224 0.224] Crate Mesh B
      (Rendering|Material|SetMaterial (Variables|Components|GetSM_Quest_PreInteraction) 0 (Variables|Design-CrateCustomization|GetVariantMaterialB))
      (Rendering|Material|SetMaterial (Variables|Components|GetSM_Quest_PostInteraction) 0 (Variables|Design-CrateCustomization|GetVariantMaterialB))
      (else
        (Rendering|Material|SetMaterial (Variables|Components|GetSM_Quest_PreInteraction) 0 (Variables|Design-CrateCustomization|GetOriginalMaterialB))
        (Rendering|Material|SetMaterial (Variables|Components|GetSM_Quest_PostInteraction) 0 (Variables|Design-CrateCustomization|GetOriginalMaterialB))))))
